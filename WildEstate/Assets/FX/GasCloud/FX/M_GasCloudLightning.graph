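# Translucent unlit, two-sided material that draws animated, stylized lightning for a gas cloud on spline-thickened strips, with a camera-axis offset.
# Panning, sine-distorted texture samples are thresholded into a hard core and soft fringe, driving layered colors and depth-faded opacity.
# Named reroutes define tiled UVs, end gradients, a travelling mask and position-based randomness per instance.
# Color gets a camera-height anti-fog subtraction and inverse eye adaptation.
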Material /WildEstate/Assets/FX/GasCloud/FX/M_GasCloudLightning
graph {
 properties {
  bAutomaticallySetUsageInEditor=false
  bHasPixelAnimation=true
  blendMode=BLEND_Translucent
  shadingModel=MSM_Unlit
  twoSided=true
  usedShadingModels=Unlit
 }
 nodes [
  n1 = Multiply(
   A=LinearInterpolate(
    A=ScalarParameter() {defaultValue=441.5891, parameterName=WIdthStart}
    B=ScalarParameter() {defaultValue=2136.136, parameterName=WIdth}
    Alpha=NamedRerouteUsage()
   )
   B=MaterialFunctionCall() {materialFunction=&ObjectScale}.Scale X
  )
  n2 = MaterialFunctionCall(WidthBase=@n1, WidthTip=@n1) {materialFunction=&SplineThicken}
  n3 = Add(A=@n2.WorldPositionOffset, B=WorldPosition() {worldPositionShaderOffset=WPT_CameraRelative}.XYZ)
  Add(
   A=@n2.WorldPositionOffset
   B=Multiply(A=Normalize(VectorInput=@n3), B=ScalarParameter() {defaultValue=-16500.3574, parameterName=CameraPush})
  )
  n4 = ScalarParameter() {defaultValue=0.0912, parameterName=Threshold}
  n5 = Constant()
  n6 = Add(A=ScalarParameter() {defaultValue=0.044, parameterName=Speed_Distortion}, B=NamedRerouteUsage())
  n7 = ComponentMask(None=Panner(Coordinate=NamedRerouteUsage(), Speed=AppendVector(A=@n5, B=@n6))) {g=true}
  n8 = Multiply(A=@n7, B=ScalarParameter() {defaultValue=3, parameterName=DistortionFequency})
  n9 = Divide(A=Add(A=Sine(None=@n8)))
  n10 = AppendVector(A=ScalarParameter() {defaultValue=0.1296, parameterName=Distortion1}, B=ScalarParameter() {defaultValue=-0.01, parameterName=Distortion2})
  n11 = ComponentMask(None=TextureCoordinate()) {r=true}
  n12 = Subtract(A=@n11) {constB=0.5}
  n13 = ScalarParameter() {defaultValue=0.138, parameterName=CenterFalloff}
  n14 = Divide(A=Abs(None=@n12), B=@n13)
  n15 = AppendVector(A=Multiply(A=Saturate(None=@n14), B=Sign(None=@n12)), B=Constant() {r=1})
  n16 = Constant()
  n17 = NamedRerouteUsage()
  n18 = Add(
   A=Multiply(A=Multiply(A=@n9, B=@n10), B=@n15)
   B=Panner(Coordinate=NamedRerouteUsage(), Speed=AppendVector(A=@n16, B=@n17)) {bFractionalPart=true}
  )
  n19 = TextureSample!(UVs=@n18, Tex, "DDX(UVs)"=DDX(Value=@n18), "DDY(UVs)"=DDY(Value=@n18), Apply View MipBias) {mipValueMode=TMVM_Derivative, texture=&/GasCloud_Ligthing_M}
  n20 = Sine(None=Multiply(A=@n7, B=ScalarParameter() {defaultValue=5, parameterName=DistortionFequency2}))
  n21 = Multiply(A=Divide(A=Add(A=@n20)), B=@n10)
  n22 = NamedRerouteUsage()
  n23 = AppendVector(A=Constant(), B=Add(A=NamedRerouteUsage(), B=ScalarParameter() {defaultValue=0.02, parameterName=Speed2}))
  n24 = TextureSample!(UVs=Add(A=Multiply(A=@n21, B=@n15), B=Panner(Coordinate=@n22, Speed=@n23) {bFractionalPart=true}), Tex, Apply View MipBias) {texture=&GasCloud_Ligthing_M}
  n25 = Multiply(A=NamedRerouteUsage(), B=Multiply(A=NamedRerouteUsage(), B=Add(A=@n19.R, B=@n24.G)))
  n26 = Saturate(None=Round(None=Add(A=@n4, B=@n25)))
  n27 = Divide(
   A=Subtract(A=@n25, B=ScalarParameter() {defaultValue=0.0504, parameterName=Threshold_translucent})
   B=ScalarParameter() {defaultValue=0.216, parameterName=Threshold_Range}
  )
  DepthFade(Opacity=Saturate(None=Add(A=@n26, B=@n27)) {desc=Opacity}, FadeDistance=ScalarParameter() {defaultValue=500, parameterName=DepthFade})
  n28 = VectorParameter() {parameterName=Color2_DarkLine, defaultValue=<rgba 0.5282 0.0928 0.5938 1>}
  n29 = VectorParameter() {parameterName=Color2, defaultValue=<rgba 0.9156 0.3542 1 1>}
  n30 = Constant()
  n31 = Add(A=NamedRerouteUsage(), B=ScalarParameter() {defaultValue=0.02, parameterName=Speed3})
  n32 = TextureSample(UVs=Panner(Coordinate=NamedRerouteUsage(), Speed=AppendVector(A=@n30, B=@n31)) {bFractionalPart=true}) {texture=&GasCloud_Ligthing_M}
  n33 = ScalarParameter() {defaultValue=-0.3652, parameterName=Threshold3}
  n34 = Saturate(None=Divide(A=Add(A=@n32.R, B=@n33), B=ScalarParameter() {defaultValue=0.02, parameterName=Threshold_Range2}))
  n35 = Round(None=Add(A=ScalarParameter() {defaultValue=0.228, parameterName=Threshold2}, B=@n32.R))
  n36 = LinearInterpolate(
   A=VectorParameter() {parameterName=Color_Translucent, defaultValue=<rgba 0.5924 0.0874 0.7708 1>}.RGB
   B=LinearInterpolate(
    A=VectorParameter() {parameterName=Color, defaultValue=<rgba 0.7844 0.2258 1 1>}.RGB
    B=LinearInterpolate(A=@n28.RGB, B=@n29.RGB, Alpha=@n34)
    Alpha=Saturate(None=@n35)
   )
   Alpha=@n26
  )
  Desaturation(
   None=MaterialFunctionCall(Hue Shift Percentage=ScalarParameter() {defaultValue=0.144, parameterName=HueShift}, Texture=@n36) {materialFunction=&HueShift}.Result
   Fraction=ScalarParameter() {parameterName=Saturation}
  )
  n37 = NamedRerouteUsage()
  n38 = ScalarParameter() {defaultValue=1.1636, parameterName=Mask_Offset}
  n39 = Add(A=NamedRerouteUsage(), B=AppendVector(A=Constant(), B=Multiply(A=@n37, B=@n38)))
  n40 = AppendVector(A=Constant(), B=Add(A=ScalarParameter() {defaultValue=0.044, parameterName=Speed_Mask}, B=NamedRerouteUsage()))
  n41 = Divide(A=ComponentMask(None=Panner(Coordinate=@n39, Speed=@n40)) {g=true}, B=ScalarParameter() {defaultValue=1.25, parameterName=Mask_Length})
  Divide(A=Add(A=Sine(None=@n41)))
  n42 = VectorParameter() {parameterName=AntiFog, defaultValue=<rgba 0.0887 0.1258 0.1562 1>}
  n43 = Constant3Vector()
  n44 = ComponentMask(None=CameraPositionWS()) {b=true}
  n45 = ScalarParameter() {defaultValue=-15000, parameterName=HeightFade_Bias}
  n46 = Saturate(None=Divide(A=Add(A=@n44, B=@n45), B=ScalarParameter() {defaultValue=20000, parameterName=HeightFade_Distance}))
  EyeAdaptationInverse(LightValueInput=Subtract(A=@n36, B=LinearInterpolate(A=@n42.RGB, B=@n43, Alpha=@n46)))
  n47 = Constant() {r=1}
  n48 = ScalarParameter() {defaultValue=1.3776, parameterName=Tiling}
  NamedRerouteDeclaration(None=Multiply(A=TextureCoordinate(), B=AppendVector(A=@n47, B=@n48))) {name=BaseUVs, nodeColor=<rgba 1 0.2578 0 1>}
  n49 = ScalarParameter() {defaultValue=0.02, parameterName=SpeedBase}
  n50 = ScalarParameter() {defaultValue=0.01, parameterName=SpeedBaseVar}
  NamedRerouteDeclaration(None=LinearInterpolate(A=@n49, B=Add(A=@n49, B=@n50) {constB=0.1}, Alpha=NamedRerouteUsage())) {name=SpeedBaseControl, nodeColor=<rgba 0 0.0625 1 1>}
  n51 = Frac(None=ComponentMask(None=TextureCoordinate()) {g=true})
  n52 = Multiply(A=Multiply(A=@n51, B=OneMinus(None=@n51))) {constB=4}
  n53 = ScalarParameter() {defaultValue=1.4193, parameterName=StartEndGradientLength2}
  NamedRerouteDeclaration(None=Saturate(None=Divide(A=@n52, B=@n53))) {name=StarteEndWidth, nodeColor=<rgba 1 0 0.5156 1>}
  n54 = ScalarParameter() {defaultValue=1.3762, parameterName=StartEndGradientLength}
  NamedRerouteDeclaration(None=Saturate(None=Divide(A=@n52, B=@n54))) {name=StarteEndFade, nodeColor=<rgba 1 0 0.5156 1>}
  MaterialExpressionTransformPosition_0 = TransformPosition(None=Constant3Vector()) {transformSourceType=TRANSFORMPOSSOURCE_Instance, transformType=TRANSFORMPOSSOURCE_World}
  n55 = Convert(None=@MaterialExpressionTransformPosition_0) {
   nodeName=Break Float3
   convertInputs [{
     type=Vector3
     expressionInput {
      expression=&self:MaterialExpressionTransformPosition_0
      inputName=None
      mask=0
      maskA=0
      maskB=0
      maskG=0
      maskR=0
      outputIndex=0
     }
    }]
   convertOutputs [{type=Scalar}, {type=Scalar}, {type=Scalar}]
  }
  n56 = Add(A=Add(A=@n55, B=@n55), B=@n55)
  NamedRerouteDeclaration(None=Frac(None=Divide(A=@n56) {constB=30})) {name=Random, nodeColor=<rgba 1 0 0.8437 1>}
  n57 = Saturate(None=Divide(A=Frac(None=@n41), B=ScalarParameter() {defaultValue=0.5406, parameterName=Mask_Length_Fraction}))
  n58 = OneMinus(None=@n57)
  NamedRerouteDeclaration(None=Multiply(A=Multiply(A=@n57, B=@n58)) {constB=4}) {name=Mask, nodeColor=<rgba 0.3828 1 0 1>}
  n59 = Frac(None=@n11)
  n60 = Multiply(A=@n59, B=OneMinus(None=@n59))
  Saturate(None=Power(Base=OneMinus(None=@n60), Exp=@n13))
 ]
}
