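# Physics asset for the rigged gas cloud test mesh in the FX folder, giving a humanoid skeleton simplified collision bodies.
# Single capsules cover the pelvis, two spine bones, head, clavicles, arms, hands, thigh twist bones, calves and feet.
# Constraints link them outward from the pelvis with uniform swing and twist limit angles.
# The calves and thigh twist bones each attach directly to the pelvis, so calves are not chained through the thighs.
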PhysicsAsset /WildEstate/Assets/FX/GasCloud/GasCloud_Test_Rigged_PhysicsAsset
bodies {
 calf_l [{shapeType=Capsule, length=23.2794, radius=15.1293, center=<-11.9302 -4.7801 0.6003>, rotation=<rot 0 134.5899 -89.2046>}]
 calf_r [{shapeType=Capsule, length=22.5502, radius=14.7322, center=<11.1722 4.8094 -0.4226>, rotation=<rot 0 140.6911 -87.9212>}]
 clavicle_l [{shapeType=Capsule, length=14.7426, radius=12.0535, center=<3.933 5.8104 1.9505>, rotation=<rot 0 11.1054 -71.9955>}]
 clavicle_r [{shapeType=Capsule, length=15.0025, radius=11.4514, center=<-6.3106 -5.8969 -4.4163>, rotation=<rot 0 6.2004 -79.9962>}]
 foot_l [{shapeType=Capsule, length=18.6453, radius=11.5675, center=<-6.2782 2.6746 0.9417>, rotation=<rot 0 -13.5147 -88.4025>}]
 foot_r [{shapeType=Capsule, length=18.5783, radius=11.5669, center=<6.4774 -2.4563 -0.487>, rotation=<rot 0 -14.7804 -83.2394>}]
 hand_l [{shapeType=Capsule, length=10.6389, radius=7.4944, center=<9.0096 -1.716 0.3563>, rotation=<rot 0 80.9186 -77.399>}]
 hand_r [{shapeType=Capsule, length=11.1172, radius=6.834, center=<-9.0198 2.3345 -0.1586>, rotation=<rot 0 -105.3542 -86.8392>}]
 head [{shapeType=Capsule, length=28.5196, radius=18.56, center=<19.6679 3.6279 0.2141>, rotation=<rot 0 -107.5284 -88.9029>}]
 lowerarm_l [{shapeType=Capsule, length=9.0255, radius=8.8596, center=<7.597 1.5962 -1.1327>, rotation=<rot 0 -157.1902 -73.0371>}]
 lowerarm_r [{shapeType=Capsule, length=8.5496, radius=8.0177, center=<-7.4345 -1.4243 -0.1653>, rotation=<rot 0 38.7738 -82.9226>}]
 pelvis [{shapeType=Capsule, length=30.1408, radius=29.1202, center=<-2.6795 -2.6802 0.2986>, rotation=<rot 15.3203 95.9301 -90>}]
 spine_03 [{shapeType=Capsule, length=21.1064, radius=17.3769, center=<8.1797 0.2934 0.0165>, rotation=<rot 0 4.5294 -86.4062>}]
 spine_05 [{shapeType=Capsule, length=12.5255, radius=9.291, center=<6.8107 3.1406 0.8682>, rotation=<rot 0 -158.0488 -88.028>}]
 thigh_twist_01_l [{shapeType=Capsule, length=16.4289, radius=12.3979, center=<3.8432 -0.5197 -0.865>, rotation=<rot 0 176.4842 -86.1337>}]
 thigh_twist_01_r [{shapeType=Capsule, length=16.6631, radius=13.0596, center=<-3.8302 0.739 0.409>, rotation=<rot 0 -178.9263 -89.7136>}]
 upperarm_l [{shapeType=Capsule, length=12.4969, radius=10.7842, center=<12.8845 0.3746 -0.7279>, rotation=<rot 0 -99.7833 -66.1568>}]
 upperarm_r [{shapeType=Capsule, length=14.3115, radius=11.0161, center=<-12.9232 -1.5322 2.9264>, rotation=<rot 0 -68.9678 -78.4594>}]
}
constraints [
 {bone1Name=calf_l, bone2Name=pelvis, swing1LimitDegrees=45, swing2LimitDegrees=45, twistLimitDegrees=45}
 {bone1Name=foot_l, bone2Name=calf_l, swing1LimitDegrees=45, swing2LimitDegrees=45, twistLimitDegrees=45}
 {bone1Name=thigh_twist_01_l, bone2Name=pelvis, swing1LimitDegrees=45, swing2LimitDegrees=45, twistLimitDegrees=45}
 {bone1Name=calf_r, bone2Name=pelvis, swing1LimitDegrees=45, swing2LimitDegrees=45, twistLimitDegrees=45}
 {bone1Name=foot_r, bone2Name=calf_r, swing1LimitDegrees=45, swing2LimitDegrees=45, twistLimitDegrees=45}
 {bone1Name=thigh_twist_01_r, bone2Name=pelvis, swing1LimitDegrees=45, swing2LimitDegrees=45, twistLimitDegrees=45}
 {bone1Name=spine_03, bone2Name=pelvis, swing1LimitDegrees=45, swing2LimitDegrees=45, twistLimitDegrees=45}
 {bone1Name=spine_05, bone2Name=spine_03, swing1LimitDegrees=45, swing2LimitDegrees=45, twistLimitDegrees=45}
 {bone1Name=head, bone2Name=spine_05, swing1LimitDegrees=45, swing2LimitDegrees=45, twistLimitDegrees=45}
 {bone1Name=clavicle_r, bone2Name=spine_05, swing1LimitDegrees=45, swing2LimitDegrees=45, twistLimitDegrees=45}
 {bone1Name=upperarm_r, bone2Name=clavicle_r, swing1LimitDegrees=45, swing2LimitDegrees=45, twistLimitDegrees=45}
 {bone1Name=lowerarm_r, bone2Name=upperarm_r, swing1LimitDegrees=45, swing2LimitDegrees=45, twistLimitDegrees=45}
 {bone1Name=hand_r, bone2Name=lowerarm_r, swing1LimitDegrees=45, swing2LimitDegrees=45, twistLimitDegrees=45}
 {bone1Name=clavicle_l, bone2Name=spine_05, swing1LimitDegrees=45, swing2LimitDegrees=45, twistLimitDegrees=45}
 {bone1Name=upperarm_l, bone2Name=clavicle_l, swing1LimitDegrees=45, swing2LimitDegrees=45, twistLimitDegrees=45}
 {bone1Name=lowerarm_l, bone2Name=upperarm_l, swing1LimitDegrees=45, swing2LimitDegrees=45, twistLimitDegrees=45}
 {bone1Name=hand_l, bone2Name=lowerarm_l, swing1LimitDegrees=45, swing2LimitDegrees=45, twistLimitDegrees=45}
]
