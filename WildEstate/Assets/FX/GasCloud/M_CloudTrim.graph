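# Unlit masked material for gas cloud trim, cut out by an inverted, scaled trim texture and usable on skeletal meshes.
# SplineThicken, a camera-axis push and a world-position sine wobble, all scaled by object Z scale, form the vertex offset.
# Color blends between two tints by pixel depth, is multiplied by a time-of-day tint and passes through EyeAdaptationInverse.
# A vertex-color random time offset is computed but does not feed the wobble.
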Material /WildEstate/Assets/FX/GasCloud/M_CloudTrim
graph {
 properties {
  bAutomaticallySetUsageInEditor=false
  bUsedWithMeshDeformer=true
  bUsedWithSkeletalMesh=true
  blendMode=BLEND_Masked
  shadingModel=MSM_Unlit
  usedShadingModels=Unlit
 }
 nodes [
  n1 = Multiply(
   A=ScalarParameter() {defaultValue=-1.8225, parameterName=Width}
   B=MaterialFunctionCall() {materialFunction=&ObjectScale}.Scale Z
  )
  n2 = Normalize(VectorInput=WorldPosition() {worldPositionShaderOffset=WPT_CameraRelative}.XYZ)
  n3 = Multiply(
   A=ScalarParameter() {defaultValue=-5.2729, parameterName=PushToCamera}
   B=MaterialFunctionCall() {materialFunction=&ObjectScale}.Scale Z
  )
  n4 = ScalarParameter() {defaultValue=0.3, parameterName=SineMovement_Amp}
  n5 = MaterialFunctionCall() {materialFunction=&ObjectScale}
  n6 = ScalarParameter() {defaultValue=11.1265, parameterName=SinSIze}
  n7 = Time()
  n8 = Add(
   A=Divide(A=WorldPosition() {worldPositionShaderOffset=WPT_ExcludeAllShaderOffsets}.XYZ, B=Multiply(A=@n6, B=@n5.Scale Z))
   B=Multiply(A=ScalarParameter() {defaultValue=0.5, parameterName=SineMovement_Speed}, B=@n7)
  )
  Add(
   A=Add(
    A=MaterialFunctionCall(WidthBase=@n1, WidthTip=@n1) {materialFunction=&SplineThicken}.WorldPositionOffset
    B=Multiply(A=@n2, B=@n3)
   )
   B=Multiply(A=Multiply(A=@n4, B=@n5.Scale Z), B=Sine(None=@n8))
  )
  Add(A=@n7, B=Multiply(A=VertexColor().R, B=ScalarParameter() {defaultValue=82.1504, parameterName=RandomTImeOffset}))
  n9 = VectorParameter() {parameterName=ColorTint_Morning, defaultValue=<rgba 1.5 1.1562 1.2481 1>}
  n10 = VectorParameter() {parameterName=ColorTint_Day, defaultValue=<rgba 1 1 1 1>}
  n11 = VectorParameter() {parameterName=ColorTint_Evening, defaultValue=<rgba 1 1 1 1>}
  n12 = VectorParameter() {parameterName=ColorTint_Night, defaultValue=<rgba 0.5312 0.8064 1 1>}
  n13 = VectorParameter() {parameterName=Color2, defaultValue=<rgba 0.3591 0.8698 0.1338 1>}
  n14 = VectorParameter() {parameterName=Color, defaultValue=<rgba 0.0541 0.0885 0.0148 1>}
  n15 = PixelDepth()
  n16 = ScalarParameter() {defaultValue=18917.4688, parameterName=Distance_Offset}
  n17 = Saturate(None=Divide(A=Subtract(A=@n15.R, B=@n16), B=ScalarParameter() {defaultValue=5385.291, parameterName=Distance_Fade}))
  EyeAdaptationInverse(LightValueInput=Multiply(
    A=MaterialFunctionCall(Color_Morning=@n9.RGB, Color_Day=@n10.RGB, Color_Evening=@n11.RGB, Color_Night=@n12.RGB) {materialFunction=&Color_Time_of_Day}.Result
    B=LinearInterpolate(A=@n13.RGB, B=@n14.RGB, Alpha=@n17)
   ))
  n18 = TextureCoordinate()
  n19 = AppendVector(A=ScalarParameter() {defaultValue=1, parameterName=ScaleX}, B=ScalarParameter() {defaultValue=154.6819, parameterName=ScaleY})
  OneMinus(None=TextureSample(UVs=Divide(A=@n18, B=@n19)) {samplerType=SAMPLERTYPE_Masks, texture=&/GasCloud_trim}.R)
 ]
}
