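# Unlit, two-sided masked gas cloud material for 2.5D cloud cards, with switches for solid body, glow and radial gradient variants.
# Sine-distorted UVs sample colour and mask, with optional normal-based two-tone shading, hue correction, camera-height anti-fog, time-of-day opacity and depth fade.
# World position offset adds sine floating, staggered by vertex colour, plus camera-facing via sprites or mesh alignment with angle-based wobble, and camera push on non-solid variants.
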Material /WildEstate/Assets/FX/GasCloud/M_GasCloud_25D
graph {
 properties {
  bAutomaticallySetUsageInEditor=false
  bUseTranslucencyVertexFog=false
  blendMode=BLEND_Masked
  shadingModel=MSM_Unlit
  twoSided=true
  usedShadingModels=Unlit
 }
 nodes [
  n1 = DotProduct(A=VertexNormalWS(), B=CameraVectorWS())
  Ceil(None=@n1)
  n2 = CollectionParameter() {
   collection=&FortniteMaterialParameters
   parameterName=SunlightDirectionalVector
  }
  ComponentMask(None=@n2) {g=true, r=true}
  n3 = MaterialFunctionCall(
   Value_Morning=ScalarParameter() {defaultValue=0.75, parameterName=Opacity_Morning}
   Value_Day=ScalarParameter() {defaultValue=1, parameterName=Opacity_Day}
   Value_Evening=ScalarParameter() {defaultValue=0.75, parameterName=Opacity_Evening}
   Value_Night=ScalarParameter() {defaultValue=0.4, parameterName=Opacity_Night}
  ) {materialFunction=&Scalar_Time_of_Day}
  n4 = Add(A=Time(), B=Multiply(A=ScalarParameter() {defaultValue=33.33, parameterName=TimeOffset}, B=VertexColor().R))
  n5 = Add(
   A=Multiply(A=TextureCoordinate(), B=ScalarParameter() {defaultValue=8, parameterName=SineUVFreqency})
   B=Multiply(A=ScalarParameter() {defaultValue=0.2, parameterName=UVDistortSpeed}, B=@n4)
  )
  n6 = Add(A=TextureCoordinate(), B=Multiply(A=ScalarParameter() {defaultValue=0.003, parameterName=UVdistortionStrg}, B=Sine(None=@n5)))
  n7 = StaticSwitchParameter(
   True=MaterialFunctionCall(
    UVs=TextureCoordinate()
    CenterPosition=Constant2Vector() {g=0.5, r=0.5}
    Radius=Constant() {r=0.5}
    Density=ScalarParameter() {defaultValue=1, parameterName=RadialGradient_Density}
   ) {materialFunction=&RadialGradientExponential}.RadialGradientExponential
   False=TextureSampleParameter2D(UVs=@n6) {parameterName=Opacity, texture=&/GasCloud_25D_Stage3_Mask}.R
  ) {parameterName=Radial Gradient}
  DepthFade(
   Opacity=Multiply(A=ScalarParameter() {defaultValue=0.2542, parameterName=Opacity}, B=Multiply(A=@n3.Result, B=@n7))
   FadeDistance=ScalarParameter() {defaultValue=5000, parameterName=DepthFade}
  )
  n8 = ScalarParameter() {defaultValue=0.04, parameterName=HueShift}
  n9 = ScalarParameter() {defaultValue=0.168, parameterName=Bias}
  n10 = DotProduct(A=Constant3Vector() {constant=<rgba 0 0 1 0>}, B=VertexNormalWS())
  n11 = Saturate(None=Ceil(None=Add(A=@n9, B=@n10)))
  n12 = LinearInterpolate(
   A=VectorParameter() {parameterName=ShadowColorSolidColor, defaultValue=<rgba 0.521 0.2086 0.8632 1>}.RGB
   B=VectorParameter() {parameterName=SolidColor, defaultValue=<rgba 0.6654 0.3095 1 1>}.RGB
   Alpha=@n11
  )
  n13 = TextureSampleParameter2D(UVs=@n6) {parameterName=Diffuse, texture=&/GasCloud_25D_Stage3}
  n14 = LinearInterpolate(A=Multiply(A=VectorParameter() {parameterName=Shadow MUlt, defaultValue=<rgba 0.776 0.776 0.776 1>}.RGB, B=@n13.RGB), B=@n13.RGB, Alpha=@n11)
  n15 = Saturate(None=LinearInterpolate(
    A=ScalarParameter() {parameterName=RemoveOutline_Low}
    B=ScalarParameter() {defaultValue=2.4091, parameterName=RemoveOutline_High}
    Alpha=@n1
   ))
  n16 = StaticSwitchParameter(True=@n12, False=StaticSwitchParameter(True=LinearInterpolate(A=@n14, B=@n12, Alpha=@n15), False=@n13.RGB) {parameterName=UseDotSShading}) {parameterName=Is Solid Body}
  n17 = Multiply(
   A=VectorParameter() {parameterName=ColorCorrection, defaultValue=<rgba 0.9804 0.9531 1 1>}.RGB
   B=Desaturation(
    None=MaterialFunctionCall(Hue Shift Percentage=@n8, Texture=@n16) {materialFunction=&HueShift}.Result
    Fraction=ScalarParameter() {defaultValue=-0.288, parameterName=Saturation}
   )
  )
  n18 = Add(A=ComponentMask(None=CameraPositionWS()) {b=true}, B=ScalarParameter() {defaultValue=-15000, parameterName=HeightFade_Bias})
  n19 = ScalarParameter() {defaultValue=20000, parameterName=HeightFade_Distance}
  n20 = LinearInterpolate(
   A=VectorParameter() {parameterName=AntiFog, defaultValue=<rgba 0.0887 0.1258 0.1562 1>}.RGB
   B=Constant3Vector()
   Alpha=Saturate(None=Divide(A=@n18, B=@n19))
  )
  EyeAdaptationInverse(LightValueInput=StaticSwitchParameter(True=VectorParameter() {parameterName=GlowColor, defaultValue=<rgba 0 2 0.4287 1>}.RGB, False=Subtract(A=@n17, B=@n20)) {parameterName=IsTranslucentGlow})
  n21 = Constant3Vector() {constant=<rgba 0 1 0 0>}
  n22 = Transform(None=@n21) {transformSourceType=TRANSFORMSOURCE_Camera}
  n23 = Constant3Vector() {constant=<rgba 0 0 1 0>}
  n24 = Transform(None=@n23) {transformSourceType=TRANSFORMSOURCE_Camera}
  MaterialFunctionCall(
   VectorToTransform=Subtract(A=WorldPosition().XYZ, B=ObjectPositionWS())
   BasisX=Normalize(VectorInput=@n22)
   BasisY=Normalize(VectorInput=CrossProduct(A=@n22, B=@n24))
   BasisZ=Normalize(VectorInput=@n24)
  ) {materialFunction=&Transform3x3Matrix}
  MaterialFunctionCall() {materialFunction=&CameraDirectionVector}
  n25 = ObjectPositionWS()
  n26 = CameraPositionWS()
  MaterialFunctionCall!(
   WorldPosition
   PivotLocation
   WS Normals
   Custom Object Basis 1=Normalize(VectorInput=Subtract(A=@n25, B=@n26))
   Custom Object Basis 2
   Custom Object Basis 3
  ) {materialFunction=/Engine/Functions/Engine_MaterialFunctions02/WorldPositionOffset/AlignMeshToTheCamera}
  Round(None=@n10)
  n27 = ScalarParameter() {defaultValue=0.248, parameterName=FloatSpeed}
  n28 = ScalarParameter() {defaultValue=0.2, parameterName=FloatSpeed2}
  n29 = AppendVector(A=Sine(None=Multiply(A=@n4, B=@n27)), B=Sine(None=Multiply(A=@n4, B=@n28)))
  n30 = Sine(None=Multiply(A=@n4, B=ScalarParameter() {defaultValue=0.3214, parameterName=FloatSpeed3}))
  n31 = Multiply(
   A=MaterialFunctionCall() {materialFunction=&ObjectScale}.Scale X
   B=Multiply(A=AppendVector(A=@n29, B=@n30), B=VectorParameter() {parameterName=FloatOffsetDitance, defaultValue=<rgba 4 4 4 1>}.RGB)
  )
  n32 = Normalize(VectorInput=WorldPosition() {worldPositionShaderOffset=WPT_CameraRelative}.XYZ)
  n33 = ScalarParameter() {parameterName=CameraPush}
  n34 = ScalarParameter() {defaultValue=85.5561, parameterName=SpriteScaleVariation}
  n35 = VertexColor()
  n36 = Multiply(
   A=Add(A=ScalarParameter() {defaultValue=50.3644, parameterName=SpriteScale}, B=Multiply(A=@n34, B=@n35.G))
   B=AppendVector(A=Constant() {r=1}, B=ScalarParameter() {defaultValue=1, parameterName=SpriteScaleAspectRatio})
  )
  n37 = MaterialFunctionCall() {materialFunction=&ObjectScale}
  n38 = Add(
   A=@n31
   B=MaterialFunctionCall!(
    Center Location=WorldPosition().XYZ
    0-1 UVs=TextureCoordinate() {coordinateIndex=1}
    XY scale=Multiply(A=@n36, B=@n37.Scale Z)
    OverridePivot
    "Normals (Optional)"
   ) {materialFunction=&Sprite}.World position offset
  )
  n39 = Add(
   A=MaterialFunctionCall(Custom Object Basis 1=@n23, Custom Object Basis 2=Constant3Vector() {constant=<rgba 1 0 0 0>}, Custom Object Basis 3=@n21) {materialFunction=&AlignMeshToTheCamera}.World Position Offset
   B=@n31
  )
  n40 = Constant3Vector() {constant=<rgba 1 0 0 0>}
  n41 = Normalize(VectorInput=Subtract(A=CameraPositionWS(), B=ObjectPositionWS()))
  n42 = Sine(None=Multiply(
    A=MaterialFunctionCall!(Normalized V1=@n40, Normalized V2=@n41) {materialFunction=/Engine/Functions/Engine_MaterialFunctions02/Math/AngleBetweenVectors}.0-1
    B=ScalarParameter() {defaultValue=20, parameterName=FrequencyDirectionWobble}
   ))
  n43 = ScalarParameter() {defaultValue=0.01, parameterName=DirectionWOobble}
  n44 = WorldPosition()
  n45 = Add(
   A=@n39
   B=RotateAboutAxis(
    NormalizedRotationAxis=Constant3Vector() {constant=<rgba 0 0 1 0>}
    RotationAngle=Multiply(A=@n42, B=@n43)
    PivotPoint=ObjectPositionWS()
    Position=Add(A=@n39, B=@n44.XYZ)
   )
  )
  StaticSwitchParameter(True=@n31, False=Add(A=Multiply(A=@n32, B=@n33), B=StaticSwitchParameter(True=@n38, False=@n45) {parameterName=UseSpriteWPO})) {parameterName=Is Solid Body}
  StaticSwitch()
  Subtract(A=@n10)
 ]
}
