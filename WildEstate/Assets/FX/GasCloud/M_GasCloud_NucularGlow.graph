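# Additive unlit glow material for the gas cloud effect, enabled for skeletal meshes.
# A tinted Glow color is scaled by a time-of-day function and compensated for eye adaptation.
# A view-versus-normal dot product raised to a power builds opacity that fades toward silhouette edges.
# Offset math combines inflation along vertex normals with a push away from the camera; the graph shows no output connections.
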Material /WildEstate/Assets/FX/GasCloud/M_GasCloud_NucularGlow
graph {
 properties {
  bAutomaticallySetUsageInEditor=false
  bUsedWithMeshDeformer=true
  bUsedWithSkeletalMesh=true
  blendMode=BLEND_Additive
  shadingModel=MSM_Unlit
  usedShadingModels=Unlit
 }
 nodes [
  n1 = Normalize(VectorInput=WorldPosition() {worldPositionShaderOffset=WPT_CameraRelativeNoOffsets}.XYZ)
  n2 = Multiply(A=VertexNormalWS(), B=ScalarParameter() {defaultValue=4477.7949, parameterName=VertexOffset})
  Add(A=Multiply(A=@n1, B=ScalarParameter() {defaultValue=12613.3809, parameterName=CameraPushBack}), B=@n2)
  n3 = Constant() {r=0.7}
  n4 = Constant() {r=1}
  n5 = Constant() {r=0.5}
  EyeAdaptationInverse(LightValueInput=Multiply(
    A=VectorParameter() {parameterName=Glow, defaultValue=<rgba 0 1 0.6193 1>}.RGB
    B=MaterialFunctionCall(Value_Morning=@n3, Value_Day=@n4, Value_Evening=@n3, Value_Night=@n5) {materialFunction=&Scalar_Time_of_Day}.Result
   ))
  n6 = Multiply(A=DotProduct(A=VertexNormalWS(), B=@n1)) {constB=-1}
  n7 = ScalarParameter() {defaultValue=7.0136, parameterName=FresnelEXP}
  Multiply(A=ScalarParameter() {defaultValue=0.632, parameterName=Opacity}, B=Power(Base=@n6, Exp=@n7))
  Multiply(A=Saturate(None=OneMinus(None=@n6)), B=@n2)
  OneMinus(None=Fresnel(ExponentIn=@n7) {baseReflectFraction=0})
 ]
}
