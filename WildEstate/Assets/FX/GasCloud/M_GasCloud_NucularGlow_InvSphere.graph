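# Unlit additive material for a green nuclear gas-cloud glow, apparently meant for an inverted sphere.
# A view-to-normal dot product raised to an exponent shapes opacity, softened by depth fade against scene geometry.
# A time-of-day function blends four tints, multiplied by the glow color and compensated for eye adaptation.
# Offset math pushes vertices away from the camera and along their normals; separate Fresnel and scaled-offset branches appear to be leftovers.
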Material /WildEstate/Assets/FX/GasCloud/M_GasCloud_NucularGlow_InvSphere
graph {
 properties {bAutomaticallySetUsageInEditor=false, blendMode=BLEND_Additive, shadingModel=MSM_Unlit, usedShadingModels=Unlit}
 nodes [
  n1 = Normalize(VectorInput=WorldPosition() {worldPositionShaderOffset=WPT_CameraRelativeNoOffsets}.XYZ)
  n2 = Multiply(A=VertexNormalWS(), B=ScalarParameter() {parameterName=VertexOffset})
  Add(A=Multiply(A=@n1, B=ScalarParameter() {defaultValue=12613.3809, parameterName=CameraPushBack}), B=@n2)
  n3 = Multiply(A=DotProduct(A=VertexNormalWS(), B=@n1)) {constB=-1}
  n4 = ScalarParameter() {defaultValue=4.8662, parameterName=FresnelEXP}
  DepthFade(Opacity=Multiply(A=ScalarParameter() {defaultValue=0.4182, parameterName=Opacity}, B=Power(Base=@n3, Exp=@n4))) {fadeDistanceDefault=5000}
  n5 = VectorParameter() {parameterName=ColorTint_Morning, defaultValue=<rgba 3 1.4081 1.8333 1>}
  n6 = VectorParameter() {parameterName=ColorTint_Day, defaultValue=<rgba 1 1 1 1>}
  n7 = VectorParameter() {parameterName=ColorTint_Evening, defaultValue=<rgba 2 0.9495 0.717 1>}
  n8 = VectorParameter() {parameterName=ColorTint_Night, defaultValue=<rgba 0.3034 0.7123 1 1>}
  EyeAdaptationInverse(LightValueInput=Multiply(
    A=MaterialFunctionCall(Color_Morning=@n5.RGB, Color_Day=@n6.RGB, Color_Evening=@n7.RGB, Color_Night=@n8.RGB) {materialFunction=&Color_Time_of_Day}.Result
    B=VectorParameter() {parameterName=Glow, defaultValue=<rgba 0 1 0.6193 1>}.RGB
   ))
  Multiply(A=Saturate(None=OneMinus(None=@n3)), B=@n2)
  OneMinus(None=Fresnel(ExponentIn=@n4) {baseReflectFraction=0})
 ]
}
